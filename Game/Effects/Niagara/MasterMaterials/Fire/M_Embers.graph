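# Unlit translucent Niagara sprite material that draws procedural ember particles with no shape texture.
# Two panning wispy noise samples, each shifted by its own dynamic-parameter offset, distort UVs feeding a radial gradient shaped by layered smoothsteps.
# That mask blends border and core colors, which lerp toward a cold color via the Cold dynamic parameter.
# Particle color tints the desaturation-adjusted emissive, and its alpha scales the mask.
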Material /Game/Effects/Niagara/MasterMaterials/Fire/M_Embers
graph {
 properties {
  bAutomaticallySetUsageInEditor=false
  bUsedWithNiagaraSprites=true
  blendMode=BLEND_Translucent
  shadingModel=MSM_Unlit
  usedShadingModels=Unlit
 }
 nodes [
  n1 = DynamicParameter() {defaultValue=<rgba 0 1 1 1>, paramNames [Cold,Offset_U,Offset_V,Param4]}
  n2 = Add(A=TextureCoordinate(), B=AppendVector(A=@n1.Offset_U, B=@n1.Offset_U))
  n3 = TextureSample(UVs=Panner(Coordinate=Multiply(A=@n2) {constB=0.2}) {speedX=0.3, speedY=0.1}) {texture=/Game/Effects/Fort_Effects/Textures/T_WispyFireNoise03}
  n4 = Add(A=TextureCoordinate(), B=AppendVector(A=@n1.Offset_V, B=@n1.Offset_V))
  n5 = TextureSample(UVs=Panner(Coordinate=Multiply(A=@n4) {constB=0.3}) {speedX=-0.25, speedY=0.12}) {texture=&T_WispyFireNoise03}
  n6 = Add(A=TextureCoordinate(), B=Multiply(A=Max(A=@n3.R, B=@n5.R)) {constB=0.25})
  n7 = OneMinus(None=Length(None=Subtract(A=@n6) {constB=0.5}) {is=class MaterialExpressionLength})
  n8 = Max(A=Multiply(A=SmoothStep(Value=@n7) {constMin=0.6}) {constB=0.6}, B=SmoothStep(Value=@n7) {constMax=0.95, constMin=0.75})
  n9 = Saturate(None=@n8)
  n10 = LinearInterpolate(
   A=LinearInterpolate(
    A=VectorParameter() {group=Color, parameterName=BorderColor, sortPriority=1, defaultValue=<rgba 0.4271 0.0515 0 1>}.RGB
    B=VectorParameter() {group=Color, parameterName=CoreColor, sortPriority=2, defaultValue=<rgba 3 0.5152 0 1>}.RGB
    Alpha=Power(Base=@n9) {constExponent=4}
   )
   B=VectorParameter() {group=Color, parameterName=ColdColor, sortPriority=2, defaultValue=<rgba 0.1198 0.0135 0.0195 1>}.RGB
   Alpha=DynamicParameter() {defaultValue=<rgba 0 1 1 1>, paramNames [Cold,Offset_U,Offset_V,Param4]}.Cold
  )
  n11 = ParticleColor()
  MaterialFunctionCall(Emissive=Desaturation(None=Multiply(A=@n10, B=@n11.RGB), Fraction=ScalarParameter() {defaultValue=0.15, group=Color, parameterName=Desaturation, sortPriority=3})) {materialFunction=&MF_FX_EmissiveAdjustments}
  Multiply(A=@n11.A, B=@n8)
 ]
}
