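# Translucent, unlit, two-sided master material for animated fire or energy effects on Niagara mesh particles and ribbons.
# Emissive blends an optional panning texture with particle color.
# Opacity builds flames from distorted scrolling noise, shaped by edge, bottom and split masks, with adjustable stepping plus depth, camera and view-angle fades.
# Niagara dynamic parameters drive UV offset and vertex displacement, combining noise waves with an optional clamped bend toward a supplied position.
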Material /Game/Effects/Niagara/MasterMaterials/Fire/M_FireEnergy_Mesh
graph {
 properties {
  bUsedWithNiagaraMeshParticles=true
  bUsedWithNiagaraRibbons=true
  blendMode=BLEND_Translucent
  pixelDepthOffsetMode=PDOM_Legacy
  previewMesh=/Game/Weapons/FORT_Melee/PIckaxe_Eternity_Female/FX/SM_PIckaxe_Eternity_Female
  shadingModel=MSM_Unlit
  twoSided=true
  usedShadingModels=Unlit
 }
 nodes [
  n1 = DynamicParameter() {desc=Parameters set by Niagara, parameterIndex=1, paramNames [X,Y,Z,WPOIntensity]}
  n2 = Clamp(
   None=Subtract(
    A=MaterialFunctionCall(X=@n1.X, Y=@n1.X, Z=@n1.X) {materialFunction=&MakeFloat3}.Result
    B=ParticlePositionWS()
   )
   Min=ScalarParameter() {defaultValue=-30, parameterName=Clamp_min}
   Max=ScalarParameter() {defaultValue=30, parameterName=Clamp_max}
  ) {maxDefault=1000, minDefault=-1000}
  n3 = OneMinus(None=TextureCoordinate())
  n4 = Multiply(A=Power(Base=ComponentMask(None=@n3) {g=true}, Exp=ScalarParameter() {defaultValue=3, parameterName=WPOBendPower}), B=@n1.WPOIntensity)
  n5 = DynamicParameter() {defaultValue=<rgba 1 1 15 1>, paramNames [Offset, Texture_mask_offset, WaveIntensity, Param4]}
  n6 = Panner(
   Coordinate=Add(A=@n5.Offset, B=TextureCoordinate())
   Time=Multiply(A=Time(), B=ScalarParameter() {defaultValue=1, group=Panner Speed, parameterName=Speed})
  ) {speedY=0.8}
  n7 = AppendVector(
   A=ComponentMask(None=TextureSample(UVs=@n6) {samplerType=SAMPLERTYPE_Normal, texture=&T_Weap_Distort_N_02}.RGB) {g=true, r=true}
   B=TextureSample(UVs=@n6) {texture=&T_Smooth_tile}.R
  )
  n8 = ComponentMask(None=TextureCoordinate()) {g=true}
  n9 = Sine(None=Add(A=Multiply(A=@n8, B=@n5.Param4), B=Time()))
  n10 = LinearInterpolate(A=@n7, B=Add(A=Multiply(A=@n9) {constB=0.19}, B=@n7), Alpha=ScalarParameter() {parameterName=Add_Noise})
  n11 = OneMinus(None=ComponentMask(None=TextureCoordinate()) {g=true})
  Add(
   A=LinearInterpolate(A=Constant(), B=LinearInterpolate(B=@n2, Alpha=@n4), Alpha=ScalarParameter() {parameterName=Use interpolate WPO})
   B=Multiply(
    A=LinearInterpolate(B=@n10, Alpha=@n11)
    B=DynamicParameter() {defaultValue=<rgba 1 1 15 1>, paramNames [Offset, Texture_mask_offset, WaveIntensity, Param4]}.WaveIntensity
   ) {constB=10}
  )
  n12 = Constant() {r=1}
  n13 = TextureSampleParameter2D(UVs=Panner(Coordinate=TextureCoordinate()) {speedY=1}) {group=Emissive, parameterName=Default_Tex, samplerType=SAMPLERTYPE_Grayscale, texture=&LowResBlurredNoise}
  n14 = ScalarParameter() {defaultValue=0.7, group=Emissive, parameterName=Use_Custom_Tex}
  MaterialFunctionCall(Emissive=Multiply(A=LinearInterpolate(A=@n12, B=@n13.R, Alpha=@n14), B=ParticleColor().RGB)) {materialFunction=&MF_FX_EmissiveAdjustments}
  n15 = Multiply(
   A=Add(
    A=DynamicParameter() {defaultValue=<rgba 1 1 15 1>, paramNames [Offset, Texture_mask_offset, WaveIntensity, Param4]}.Offset
    B=TextureCoordinate() {uTiling=0.5, vTiling=0.9}
   )
   B=ComponentMask(None=VectorParameter() {group=Fire Shape, parameterName=Fire_Micro_UVScale, sortPriority=8, defaultValue=<rgba 2 1 0 0>}.RGB) {g=true, r=true}
  )
  MaterialFunctionCall(XY=@n15) {materialFunction=&Swizzle}
  n16 = Add(
   A=DynamicParameter() {defaultValue=<rgba 1 1 15 1>, paramNames [Offset, Texture_mask_offset, WaveIntensity, Param4]}.Offset
   B=TextureCoordinate() {vTiling=1.5}
  )
  n17 = ComponentMask(None=VectorParameter() {group=Distortion, parameterName=Distortion UV Scale, sortPriority=3, defaultValue=<rgba 1 1 0 0>}.RGB) {g=true, r=true}
  n18 = Time()
  n19 = ScalarParameter() {defaultValue=1, group=Panner Speed, parameterName=Speed}
  n20 = TextureSampleParameter2D(UVs=Panner(Coordinate=Multiply(A=@n16, B=@n17), Time=Multiply(A=@n18, B=@n19)) {speedY=1}) {
   group=Distortion
   parameterName=Distortion Normal
   samplerType=SAMPLERTYPE_Normal
   sortPriority=2
   texture=&T_Weap_Distort_N_02
  }
  n21 = TextureCoordinate()
  n22 = VectorParameter() {group=Fire Shape, parameterName=Fire_MacroShape_UVScale, sortPriority=5, defaultValue=<rgba 1 1 0 0>}
  n23 = ComponentMask(None=@n22.RGB) {g=true, r=true}
  n24 = LinearInterpolate(A=ComponentMask(None=@n20.RGB) {g=true, r=true}, Alpha=ComponentMask(None=Multiply(A=@n21, B=@n23)) {g=true}) {constB=0}
  n25 = ScalarParameter() {defaultValue=3, group=Distortion, parameterName=UV_Distortion_Divsor, sortPriority=1}
  n26 = Multiply(A=Time(), B=ScalarParameter() {defaultValue=1, group=Panner Speed, parameterName=Speed})
  n27 = TextureSampleParameter2D(UVs=Add(A=Divide(A=@n24, B=@n25) {constB=20}, B=Panner(Coordinate=@n15, Time=@n26) {speedX=0.2, speedY=1.2})) {
   group=Fire Shape
   parameterName=Fire Micro Texture
   sortPriority=6
   texture=&T_Inky_Smoke_Tile
  }
  n28 = ChannelMaskParameter(None=@n27.RGBA) {group=Fire Shape, parameterName=Fire Micro Mask, sortPriority=7}
  n29 = TextureSample() {samplerType=SAMPLERTYPE_LinearGrayscale, texture=/Engine/Functions/Engine_MaterialFunctions02/ExampleContent/Textures/LowResBlurredNoise}
  n30 = ScalarParameter() {defaultValue=0.2, parameterName=Fire Micro lerp}
  n31 = MaterialFunctionCall(Float2=Multiply(A=@n21, B=@n23)) {materialFunction=&BreakOutFloat2Components}
  n32 = Multiply(A=@n31.R) {constB=0.5}
  n33 = ScalarParameter() {defaultValue=1, group=Fire Shape, parameterName=Edge Fade Value Top, sortPriority=3}
  n34 = ScalarParameter() {defaultValue=1, group=Fire Shape, parameterName=Edge Fade Value Bottom, sortPriority=4}
  n35 = Multiply(A=Power(Base=Sine(None=@n32), Exp=LinearInterpolate(A=@n33, B=@n34, Alpha=@n31.G) {constA=6, constB=2}) {constExponent=4}, B=@n31.G)
  n36 = VectorParameter() {group=BottomFade, parameterName=Bottom Fade_SphereMask_A_Scale, sortPriority=2, defaultValue=<rgba 1.2 1 0 0>}
  n37 = VectorParameter() {group=BottomFade, parameterName=Bottom Fade_SphereMask_B, sortPriority=3, defaultValue=<rgba 0.6 0.05 0 0>}
  n38 = SphereMask(
   A=Multiply(A=TextureCoordinate(), B=ComponentMask(None=@n36.RGB) {g=true, r=true})
   B=ComponentMask(None=@n37.RGB) {g=true, r=true}
   Radius=ScalarParameter() {defaultValue=0.95, group=BottomFade, parameterName=BottomFade_Radius, sortPriority=1}
  ) {attenuationRadius=0.95, hardnessPercent=0}
  n39 = LinearInterpolate(B=Saturate(None=Multiply(A=@n38) {constB=15}), Alpha=ScalarParameter() {defaultValue=1, group=BottomFade, parameterName=Bottom Fade}) {constA=1}
  n40 = Power(
   Base=Multiply(A=LinearInterpolate(A=@n28.R, B=@n29.R, Alpha=@n30) {constAlpha=0.2}, B=Multiply(A=@n35, B=@n39))
   Exp=LinearInterpolate(Alpha=@n31.G) {constA=2}
  )
  n41 = MaterialFunctionCall(XY=TextureCoordinate() {vTiling=0.5}) {materialFunction=&Swizzle}
  n42 = Multiply(A=ComponentMask(None=TextureSample() {samplerType=SAMPLERTYPE_Normal, texture=&T_Weap_Distort_N_02}.RGB) {g=true, r=true}) {constB=0.14}
  n43 = TextureSample(UVs=Panner(Coordinate=Add(A=@n41.YX, B=@n42)) {speedX=0.4}) {samplerType=SAMPLERTYPE_Grayscale, texture=&T_Smoke_Trail_Soft}
  n44 = Multiply(A=Multiply(A=Power(Base=@n43.R) {constExponent=1.6}) {constB=3}, B=@n40)
  n45 = ScalarParameter() {parameterName=Add_Noise}
  n46 = ScalarParameter() {defaultValue=2, group=Fire Shape, parameterName=Fire Split Contrast, sortPriority=1}
  n47 = OneMinus(None=Sine(None=Multiply(A=@n31.R) {constB=0.45}))
  n48 = LinearInterpolate(Alpha=@n20.R) {constA=0.5, constB=1.5}
  n49 = Round(None=Add(A=Power(Base=@n31.G, Exp=@n46), B=Power(Base=@n47, Exp=@n48) {constExponent=0.5}))
  n50 = ScalarParameter() {defaultValue=0.9, group=Fire Shape, parameterName=Fire Split}
  n51 = Multiply(
   A=Multiply(A=LinearInterpolate(A=@n40, B=@n44, Alpha=@n45), B=LinearInterpolate(B=@n49, Alpha=@n50) {constA=1, constB=0})
   B=ScalarParameter() {defaultValue=15, group=Opacity, parameterName=OpacityMaskMultiplier, sortPriority=1}
  ) {constB=3}
  n52 = Floor(None=@n51)
  n53 = ScalarParameter() {defaultValue=1, parameterName=Floor}
  n54 = Multiply(
   A=DepthFade(
    Opacity=LinearInterpolate(A=@n52, B=@n51, Alpha=@n53)
    FadeDistance=ScalarParameter() {defaultValue=5, group=Opacity, parameterName=DepthFadeDistance}
   )
   B=ParticleColor().A
  )
  n55 = MaterialFunctionCall(Distance=ScalarParameter() {defaultValue=150, group=Camera Fade, parameterName=Near Camera Distance}) {materialFunction=&Near_Camera_Fade}
  Multiply(
   A=LinearInterpolate(A=@n54, B=Multiply(A=@n54, B=@n55.Result), Alpha=ScalarParameter() {group=Camera Fade, parameterName=Use Camera Fade})
   B=MaterialFunctionCall("FadeSensitivity (optional)"=ScalarParameter() {parameterName=FadeBasedOnViewAngle}) {materialFunction=&FadeBasedOnViewAngle}.Result
  )
 ]
}
