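# Translucent, unlit, two-sided master material for stylized fire on Niagara mesh particles.
# A sine UV shape times distorted panning noise forms the flame, with Floor blends hardening edges for emissive core and opacity.
# Color layers gradients, curve atlas and core tints; opacity adds smoke, depth and view-angle fades, and an optional world-space sphere mask.
# Panned noise drives camera-distance-scaled vertex waves, with optional bending via Niagara dynamic parameters.
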Material /Game/Effects/Niagara/MasterMaterials/Fire/M_FireMesh
graph {
 properties {
  bAutomaticallySetUsageInEditor=false
  bUsedWithNiagaraMeshParticles=true
  blendMode=BLEND_Translucent
  shadingModel=MSM_Unlit
  twoSided=true
  usedShadingModels=Unlit
 }
 nodes [
  n1 = DynamicParameter() {desc=Parameters set by Niagara, parameterIndex=1, paramNames [X,Y,Z,WPOIntensity]}
  n2 = Subtract(
   A=MaterialFunctionCall(X=@n1.X, Y=@n1.X, Z=@n1.X) {materialFunction=&MakeFloat3}.Result
   B=ParticlePositionWS()
  )
  n3 = ScalarParameter() {defaultValue=-100, group=WPO, parameterName=Clamp_min}
  n4 = ScalarParameter() {defaultValue=100, group=WPO, parameterName=Clamp_max}
  n5 = TextureCoordinate()
  n6 = Power(Base=ComponentMask(None=OneMinus(None=@n5)) {g=true}, Exp=ScalarParameter() {defaultValue=3, group=WPO, parameterName=WPOBendPower})
  n7 = StaticSwitchParameter(
   True=LinearInterpolate(B=Clamp(None=@n2, Min=@n3, Max=@n4) {maxDefault=1000, minDefault=-1000}, Alpha=Multiply(A=@n6, B=@n1.WPOIntensity))
   False=Constant()
  ) {group=WPO, parameterName=Use_Interpolate WPO}
  n8 = ScalarParameter() {defaultValue=20, group=WPO, parameterName=Lod_low_WPO_intensity}
  n9 = ScalarParameter() {defaultValue=80, group=WPO, parameterName=WPO_intensity}
  n10 = Constant() {r=1700}
  n11 = Saturate(None=OneMinus(None=MaterialFunctionCall(Distance=@n10) {materialFunction=&Near_Camera_Fade}.Result))
  n12 = Add(A=TextureCoordinate(), B=NamedRerouteUsage())
  n13 = VectorParameter() {group=WPO, parameterName=WPO_UVScale, defaultValue=<rgba 0 1 0 0>}
  n14 = Add(A=Multiply(A=Add(A=@n12) {constB=0}, B=ComponentMask(None=@n13.RGB) {g=true, r=true})) {constB=0}
  n15 = ComponentMask(None=VectorParameter() {group=WPO, parameterName=WPO_Speed, defaultValue=<rgba 0 0.5 0 0>}.RGB) {g=true, r=true}
  n16 = LinearInterpolate(Alpha=TextureSample(UVs=Panner(Coordinate=@n14, Speed=@n15) {speedY=0.5}) {samplerType=SAMPLERTYPE_Normal, texture=&T_Weap_Distort_N_02}.R) {constB=1.5}
  n17 = OneMinus(None=ComponentMask(None=TextureCoordinate()) {g=true})
  n18 = ScalarParameter() {defaultValue=1.5, group=WPO, parameterName=WPOGradientPower}
  n19 = Saturate(None=Multiply(A=Power(Base=@n17, Exp=@n18) {constExponent=1.5}))
  n20 = Transform(None=Multiply(A=Multiply(A=@n16, B=@n19), B=VectorParameter() {group=WPO, parameterName=WPO_direction, defaultValue=<rgba 1 0 0 1>}.RGB))
  n21 = Add(A=Add(A=TextureCoordinate(), B=NamedRerouteUsage())) {constB=0}
  n22 = ComponentMask(None=Constant3Vector() {constant=<rgba 0 1 0 0>}) {g=true, r=true}
  n23 = Panner(
   Coordinate=Add(A=Multiply(A=@n21, B=@n22)) {constB=0}
   Speed=ComponentMask(None=Constant3Vector() {constant=<rgba 0 0.5 0 0>}) {g=true, r=true}
  ) {speedY=0.5}
  n24 = TextureCoordinate()
  n25 = Power(Base=OneMinus(None=ComponentMask(None=@n24) {g=true}), Exp=Constant() {r=1.5}) {constExponent=1.5}
  n26 = Multiply(
   A=LinearInterpolate(Alpha=TextureSample(UVs=@n23) {samplerType=SAMPLERTYPE_Normal, texture=&T_Weap_Distort_N_02}.R) {constB=1.5}
   B=Saturate(None=Multiply(A=@n25))
  )
  n27 = VectorParameter() {group=WPO, parameterName=WPO_direction2, defaultValue=<rgba 0 1.5 1.5 1>}
  n28 = LinearInterpolate(B=Transform(None=Multiply(A=@n26, B=@n27.RGB)), Alpha=ScalarParameter() {parameterName=Added Second Direction})
  n29 = Multiply(
   A=Multiply(A=Add(A=@n20, B=@n28), B=LinearInterpolate(A=@n8, B=@n9, Alpha=@n11)) {constB=80}
   B=DynamicParameter() {paramNames [Offset, Texture_mask_offset, WaveIntensity, UScale]}.WaveIntensity
  )
  n30 = TextureCoordinate() {uTiling=0.5}
  n31 = LinearInterpolate(Alpha=TextureSample(UVs=Panner(Coordinate=@n30) {speedY=0.5}) {samplerType=SAMPLERTYPE_Normal, texture=&T_Weap_Distort_N_02}.R) {constB=1.5}
  n32 = OneMinus(None=ComponentMask(None=TextureCoordinate()) {g=true})
  n33 = Saturate(None=Multiply(A=Power(Base=@n32) {constExponent=1.5}))
  n34 = Transform(None=Multiply(A=Multiply(A=@n31, B=@n33), B=Constant3Vector() {constant=<rgba 0 1 0 0>}))
  n35 = Constant() {r=15}
  n36 = VertexColor()
  n37 = DynamicParameter() {parameterIndex=2, paramNames [0, VertexColorControlZ, VertexColorControlX, Param4]}
  n38 = Constant3Vector() {constant=<rgba 1 0 0 0>}
  n39 = Add(
   A=Multiply(A=Constant3Vector() {constant=<rgba 0 0 17 0>}, B=@n37.VertexColorControlZ)
   B=Multiply(A=@n37.VertexColorControlX, B=Transform(None=@n38))
  )
  n40 = Add(A=@n7, B=Add(A=Multiply(A=@n34, B=@n35), B=Multiply(A=@n36.R, B=@n39)))
  n41 = LinearInterpolate(B=@n29, Alpha=ScalarParameter() {group=WPO, parameterName=LerpNoise to WorldZ})
  LinearInterpolate(
   B=LinearInterpolate(A=Add(A=@n7, B=@n29), B=Add(A=@n40, B=@n41), Alpha=ScalarParameter() {parameterName=Use_WorldZ_WPO})
   Alpha=ScalarParameter() {defaultValue=1, parameterName=UseWPO}
  ) {constB=0}
  NamedRerouteDeclaration(None=Multiply(A=ScalarParameter() {defaultValue=1, group=Panner Speed, parameterName=Speed 2}, B=Time())) {name=Speed 2 Param, nodeColor=<rgba 0 1 0.4141 1>}
  NamedRerouteDeclaration(None=AppendVector(A=Constant(), B=DynamicParameter() {paramNames [Offset, Texture_mask_offset, WaveIntensity, UScale]}.Offset)) {name=CustomOffset, nodeColor=<rgba 1 0.7734 0 1>}
  n42 = TextureCoordinate() {uTiling=0.7, vTiling=0.8}
  n43 = NamedRerouteUsage()
  n44 = DynamicParameter() {paramNames [Offset, Texture_mask_offset, WaveIntensity, UScale]}
  n45 = Time()
  n46 = ScalarParameter() {group=Base_color, parameterName=Loop_panner}
  n47 = TextureSampleParameter2D(UVs=Panner(Coordinate=Add(A=@n42, B=@n43), Time=LinearInterpolate(A=@n44.Texture_mask_offset, B=@n45, Alpha=@n46)) {speedY=0.2}) {group=Base_color, parameterName=Base_color_lerp_mask, texture=&T_Inky_Smoke_Tile}
  n48 = TextureCoordinate()
  n49 = Constant2Vector() {g=0.5, r=0.5}
  n50 = Multiply(
   A=LinearInterpolate(
    A=VectorParameter() {group=Base_color, parameterName=Base_Color1, defaultValue=<rgba 0.8499 0.0466 1.0417 1>}.RGB
    B=VectorParameter() {group=Base_color, parameterName=Base_Color2, defaultValue=<rgba 0 1.2585 21.3542 1>}.RGB
    Alpha=@n47.R
   )
   B=Power(
    Base=SphereMask(A=@n48, B=@n49) {attenuationRadius=0.8, hardnessPercent=0}
    Exp=ScalarParameter() {defaultValue=3, group=Base_color, parameterName=Base_Texture_power}
   ) {constExponent=3}
  )
  n51 = CurveAtlasRowParameter(CurveTime=ComponentMask(None=TextureCoordinate()) {g=true}) {
   atlas=/Game/Athena/LinearColorCurves/VFX_CurveAtlas
   curve=/Game/Accessories/FORT_Backpacks/Backpack_F_MED_Tower_Sentinel/FX/Curve_Tower_Sentinel_Female
   defaultValue=15
   parameterName=Global_color
  }
  n52 = TextureCoordinate()
  n53 = TextureCoordinate()
  n54 = VectorParameter() {group=Fire Shape, parameterName=Fire_MacroShape_UVScale, sortPriority=5, defaultValue=<rgba 1 1 0 0>}
  n55 = ComponentMask(None=@n54.RGB) {g=true, r=true}
  n56 = MaterialFunctionCall(Float2=Multiply(A=@n53, B=@n55)) {materialFunction=&BreakOutFloat2Components}
  n57 = Sine(None=Multiply(A=@n56.R) {constB=0.5})
  n58 = LinearInterpolate(
   A=ScalarParameter() {defaultValue=2, group=Fire Shape, parameterName=Edge Fade Value Top, sortPriority=3}
   B=ScalarParameter() {defaultValue=1, group=Fire Shape, parameterName=Edge Fade Value Bottom, sortPriority=4}
   Alpha=@n56.G
  ) {constA=6, constB=2}
  n59 = Multiply(A=TextureCoordinate(), B=ComponentMask(None=Constant3Vector() {constant=<rgba 1.2 1 0 0>}) {g=true, r=true})
  n60 = ComponentMask(None=Constant3Vector() {constant=<rgba 0.6 0.05 0 0>}) {g=true, r=true}
  n61 = ScalarParameter() {defaultValue=0.95, group=BottomFade, parameterName=BottomFade_Radius, sortPriority=1}
  n62 = Saturate(None=Multiply(
    A=SphereMask(A=@n59, B=@n60, Radius=@n61) {attenuationRadius=0.95, hardnessPercent=0}
    B=ScalarParameter() {defaultValue=15, parameterName=SphereMask_Multiply}
   ) {constB=15})
  n63 = OneMinus(None=ComponentMask(None=TextureCoordinate()) {g=true})
  n64 = Saturate(None=Multiply(A=Power(Base=@n63) {constExponent=1.5}) {constB=10})
  n65 = ScalarParameter() {group=BottomFade, parameterName=UseGradientBottomMask}
  n66 = Multiply(A=TextureCoordinate(), B=ComponentMask(None=Constant3Vector() {constant=<rgba 1 1 0 0>}) {g=true, r=true})
  n67 = ComponentMask(None=Constant3Vector() {constant=<rgba 0.5 1.45 0 0>}) {g=true, r=true}
  n68 = Constant() {r=2}
  n69 = Constant() {r=2.33}
  n70 = StaticBool()
  n71 = OneMinus(None=Multiply(A=MaterialFunctionCall(UVs=@n66, CenterPosition=@n67, Radius=@n68, Density=@n69, Invert Density=@n70) {materialFunction=&RadialGradientExponential}.RadialGradientExponential))
  n72 = Constant() {r=0.1}
  n73 = Multiply(A=Multiply(A=MaterialFunctionCall(In=@n71, Contrast=@n72) {materialFunction=&CheapContrast}.Result)) {constB=15}
  n74 = Multiply(
   A=Multiply(A=Power(Base=@n57, Exp=@n58) {constExponent=4}, B=@n56.G)
   B=LinearInterpolate(
    A=LinearInterpolate(A=@n62, B=@n64, Alpha=@n65)
    B=Saturate(None=@n73)
    Alpha=ScalarParameter() {group=BottomFade, parameterName=UseInvertSphereBottomMask}
   )
  )
  n75 = Multiply(
   A=ComponentMask(None=VectorParameter() {group=Distortion, parameterName=Distortion UV Scale, sortPriority=3, defaultValue=<rgba 1 1 0 0>}.RGB) {g=true, r=true}
   B=TextureCoordinate() {vTiling=1.5}
  )
  n76 = NamedRerouteUsage()
  n77 = Panner(
   Coordinate=Multiply(A=Add(A=@n75, B=@n76))
   Time=Multiply(A=Time(), B=ScalarParameter() {defaultValue=1, group=Panner Speed, parameterName=DisturbSpeed})
  ) {speedY=1}
  n78 = LinearInterpolate(
   A=ComponentMask(None=TextureSample(UVs=@n77) {samplerType=SAMPLERTYPE_Normal, texture=&T_Weap_Distort_N_02}.RGB) {g=true, r=true}
   Alpha=ComponentMask(None=Multiply(A=@n53, B=@n55)) {g=true}
  ) {constB=0}
  n79 = Constant() {r=3}
  n80 = Add(A=TextureCoordinate() {uTiling=0.9, vTiling=0.8}, B=NamedRerouteUsage())
  n81 = Multiply(
   A=ComponentMask(None=VectorParameter() {group=Distortion, parameterName=Fire Micro UV Scale, sortPriority=3, defaultValue=<rgba 1 1 0 0>}.RGB) {g=true, r=true}
   B=AppendVector(A=DynamicParameter() {paramNames [Offset, Texture_mask_offset, WaveIntensity, UScale]}.UScale, B=Constant() {r=1})
  )
  n82 = Add(A=Multiply(A=Multiply(A=@n80, B=@n81))) {constB=0}
  n83 = Multiply(A=Time(), B=ScalarParameter() {defaultValue=1, group=Panner Speed, parameterName=DisturbSpeed})
  n84 = TextureSampleParameter2D(UVs=Add(A=Divide(A=@n78, B=@n79) {constB=20}, B=Panner(Coordinate=@n82, Time=@n83) {speedY=1.2})) {parameterName=Fire Micro Texture, samplerType=SAMPLERTYPE_LinearColor, texture=&T_NoisePack_LinearColor}
  n85 = ChannelMaskParameter(None=@n84.RGBA) {group=Fire Shape, parameterName=Fire Micro Mask, sortPriority=7}
  n86 = Multiply(A=Power(Base=Multiply(A=@n85.R, B=@n74), Exp=LinearInterpolate(Alpha=@n56.G) {constA=2}))
  n87 = Multiply(A=@n86, B=Constant() {r=10}) {constB=3}
  n88 = LinearInterpolate(A=Floor(None=@n87), B=@n87, Alpha=ScalarParameter() {defaultValue=0.3, group=Fire_center, parameterName=Fire_center_Hard_edge})
  n89 = TextureCoordinate() {uTiling=0.5, vTiling=0.5}
  n90 = NamedRerouteUsage()
  n91 = TextureSample(UVs=Panner(Coordinate=Add(A=@n89, B=@n90), Time=NamedRerouteUsage()) {speedY=0.3}) {samplerType=SAMPLERTYPE_LinearColor, texture=/Game/Global/Textures/Default/Noise/T_Noise01_Inverse}
  n92 = VectorParameter() {group=Fire_center, parameterName=Fire_center_color02, defaultValue=<rgba 0 11.3982 50 1>}
  n93 = LinearInterpolate(
   B=Multiply(A=@n88, B=Multiply(A=@n91.R, B=@n92.RGB))
   Alpha=Power(Base=ComponentMask(None=@n52) {g=true}, Exp=ScalarParameter() {defaultValue=18, group=Fire_center, parameterName=Fire_center_power}) {constExponent=18}
  )
  n94 = Power(Base=ComponentMask(None=TextureCoordinate()) {g=true}) {constExponent=3}
  n95 = Add(A=TextureCoordinate() {uTiling=0.5, vTiling=0.5}, B=NamedRerouteUsage())
  n96 = NamedRerouteUsage()
  n97 = Multiply(
   A=Saturate(None=LinearInterpolate(B=@n88, Alpha=@n94))
   B=TextureSample(UVs=Panner(Coordinate=@n95, Time=@n96) {speedY=0.3}) {samplerType=SAMPLERTYPE_LinearColor, texture=/Game/Characters/Player/Male/Medium/Bodies/M_MED_SRDrift/Textures/T_ElectricNoise}.R
  )
  n98 = Desaturation(
   None=Multiply(A=VectorParameter() {group=Fire_center, parameterName=Fire_center_color, defaultValue=<rgba 7.3424 1.899 20 1>}.RGB, B=Power(Base=@n97))
   Fraction=ScalarParameter() {defaultValue=-0.5, group=Fire_center, parameterName=Fire_center_color_purple_Desaturation}
  )
  n99 = Add(
   A=LinearInterpolate(A=@n50, B=Multiply(A=@n50, B=@n51), Alpha=ScalarParameter() {defaultValue=1, parameterName=MultColorCurveForEmissive})
   B=Multiply(A=LinearInterpolate(A=@n93, B=@n98)) {constB=3}
  ) {constB=0}
  n100 = Multiply(A=ParticleColor().RGB, B=@n99)
  n101 = ScalarParameter() {parameterName=Use_particle_color}
  n102 = Desaturation(
   None=Multiply(A=LinearInterpolate(A=@n99, B=@n100, Alpha=@n101), B=ScalarParameter() {defaultValue=1, parameterName=Glider_appear_brightness})
   Fraction=ScalarParameter() {parameterName=Final_Desaturation}
  )
  Reroute(None=MaterialFunctionCall(Emissive=Multiply(A=@n102)) {materialFunction=&MF_FX_EmissiveAdjustments}.Result)
  n103 = ScalarParameter() {defaultValue=-0.3, group=Opacity, parameterName=Mask_Dark}
  n104 = TextureCoordinate()
  n105 = Sine(None=ComponentMask(None=@n104) {r=true}) {period=2}
  n106 = LinearInterpolate(Alpha=ComponentMask(None=@n104) {g=true}) {constA=50, constB=10}
  n107 = ComponentMask(None=VectorParameter() {group=Opacity, parameterName=Mask_UV_Scale, defaultValue=<rgba 1 1 0 1>}.RGB) {g=true, r=true}
  n108 = Add(A=Multiply(A=TextureCoordinate(), B=@n107), B=AppendVector(A=Constant() {r=0.6}, B=Constant())) {constB=0.6}
  n109 = NamedRerouteUsage()
  n110 = NamedRerouteUsage()
  n111 = TextureSample(UVs=Panner(Coordinate=Add(A=@n108, B=@n109), Time=@n110) {speedY=0.5}) {texture=/Game/CharClasses/Outlander/Gadgets/AntiMaterialCharge/Materials/T_AntiMaterialCharge_Dissolve_01}
  n112 = Multiply(A=@n107, B=TextureCoordinate() {uTiling=0.8})
  n113 = TextureSample(UVs=Panner(Coordinate=Add(A=@n112, B=@n109) {constB=0}, Time=@n110) {speedY=0.5}) {texture=&T_AntiMaterialCharge_Dissolve_01}
  n114 = Add(A=Saturate(None=Power(Base=@n105, Exp=@n106) {constExponent=10}), B=OneMinus(None=LinearInterpolate(A=@n111.R, B=@n113.R)))
  n115 = Constant() {r=0.3}
  n116 = Add(A=Add(A=TextureCoordinate() {uTiling=0.5, vTiling=0.9}, B=NamedRerouteUsage())) {constB=0}
  n117 = NamedRerouteUsage()
  n118 = Power(Base=TextureSample(UVs=Panner(Coordinate=@n116, Time=@n117) {speedY=0.5}) {texture=&T_ShieldCaustics}.R) {constExponent=0.6}
  n119 = Saturate(None=LinearInterpolate(Alpha=LinearInterpolate(Alpha=@n118) {constB=3}) {constA=-0.3})
  n120 = ComponentMask(None=TextureCoordinate()) {g=true}
  n121 = TextureCoordinate()
  n122 = Sine(None=ComponentMask(None=@n121) {r=true}) {period=2}
  n123 = LinearInterpolate(Alpha=ComponentMask(None=@n121) {g=true}) {constA=50, constB=0.5}
  n124 = Saturate(None=Multiply(A=Power(Base=@n120) {constExponent=4}, B=Power(Base=@n122, Exp=@n123) {constExponent=10}))
  n125 = TextureCoordinate()
  n126 = Multiply(A=OneMinus(None=ComponentMask(None=@n125) {g=true})) {constB=3}
  n127 = Multiply(A=Multiply(A=Add(A=@n119, B=@n124) {constB=0}, B=Saturate(None=@n126))) {constB=0.3}
  n128 = Multiply(A=@n86, B=ScalarParameter() {defaultValue=6, group=Opacity, parameterName=OpacityMaskMultiplier, sortPriority=1}) {constB=3}
  n129 = Multiply(
   A=StaticSwitchParameter(True=Add(A=@n115, B=@n127), False=LinearInterpolate(A=@n103, Alpha=@n114) {constA=-0.3}) {group=Opacity, parameterName=Use_detail_smoke}
   B=LinearInterpolate(A=Floor(None=@n128), B=@n128, Alpha=ScalarParameter() {defaultValue=1, parameterName=Floor})
  ) {constA=1}
  n130 = ComponentMask(None=@n129) {r=true}
  n131 = Multiply(A=@n130) {constB=3}
  n132 = Saturate(None=LinearInterpolate(A=@n130, B=LinearInterpolate(Alpha=@n131)))
  n133 = Multiply(
   A=Multiply(
    A=StaticSwitchParameter(True=@n132, False=@n129) {group=Opacity, parameterName=Use_detail_smoke}
    B=ScalarParameter() {defaultValue=1, group=Opacity, parameterName=Global_Opacity}
   )
   B=MaterialFunctionCall("FadeSensitivity (optional)"=ScalarParameter() {group=Opacity, parameterName=FadeBasedOnViewAngle}) {materialFunction=&FadeBasedOnViewAngle}.Result
  )
  n134 = ScalarParameter() {defaultValue=30, group=Opacity, parameterName=Depth_fade}
  n135 = Multiply(A=Multiply(A=ParticleColor().A, B=DepthFade(Opacity=@n133, FadeDistance=@n134)))
  n136 = DynamicParameter() {parameterIndex=3, paramNames [X,Y,Z,none]}
  n137 = SphereMask(
   A=WorldPosition().XYZ
   B=MaterialFunctionCall(X=@n136.X, Y=@n136.X, Z=@n136.X) {materialFunction=&MakeFloat3}.Result
   Radius=ScalarParameter() {defaultValue=13, group=CircularMask, parameterName=CircularMask}
   Hardness=ScalarParameter() {defaultValue=0.8, group=CircularMask, parameterName=Hardness}
  )
  n138 = LinearInterpolate(A=OneMinus(None=@n137), B=@n137, Alpha=ScalarParameter() {group=CircularMask, parameterName=UseInvert})
  n139 = ScalarParameter() {group=CircularMask, parameterName=UseCircularMask}
  Reroute(None=Multiply(A=Multiply(A=@n135), B=LinearInterpolate(B=@n138, Alpha=@n139) {constA=1}))
 ]
}
