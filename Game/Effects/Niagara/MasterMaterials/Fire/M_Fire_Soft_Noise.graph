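# Translucent unlit Niagara sprite master material for soft, noise-driven fire, with dynamic parameters for tiling, panning, erosion, HDR intensity and depth fade.
# Panned noise green channel pushes UVs vertically into an off-center exponential radial gradient, multiplies it, then Erode is subtracted.
# That mask scales HDR-remapped particle color for emissive, and particle alpha for depth-faded opacity.
# Tiling, offset and speed use only their X parameters for both axes.
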Material /Game/Effects/Niagara/MasterMaterials/Fire/M_Fire_Soft_Noise
graph {
 properties {
  bUseTranslucencyVertexFog=false
  bUsedWithNiagaraSprites=true
  blendMode=BLEND_Translucent
  pixelDepthOffsetMode=PDOM_Legacy
  shadingModel=MSM_Unlit
  usedShadingModels=Unlit
 }
 nodes [
  n1 = DynamicParameter() {paramNames [TileXOffset,TileYOffset,TileX,TileY]}
  n2 = Multiply(A=TextureCoordinate(), B=AppendVector(A=@n1.TileX, B=@n1.TileX))
  n3 = DynamicParameter() {paramNames [TileXOffset,TileYOffset,TileX,TileY]}
  n4 = AppendVector(A=@n3.TileXOffset, B=@n3.TileXOffset)
  n5 = DynamicParameter() {parameterIndex=1, defaultValue=<rgba 0 1 1 25>, paramNames [Erode,SpeedX,SpeedY,HDRIntensity]}
  n6 = TextureSampleParameter2D(UVs=Panner(Coordinate=Add(A=@n2, B=@n4), Speed=AppendVector(A=@n5.SpeedX, B=@n5.SpeedX)) {speedY=1}) {parameterName=NoiseTexture, texture=/Game/Effects/FogVolume/T_TileNoise}
  n7 = Constant()
  n8 = Constant() {r=0.5}
  n9 = Add(A=TextureCoordinate(), B=Multiply(A=@n6.G, B=AppendVector(A=@n7, B=@n8)))
  n10 = Constant2Vector() {g=0.7, r=0.5}
  n11 = ScalarParameter() {defaultValue=0.45, parameterName="Radius (S)"}
  n12 = Subtract(
   A=Multiply(
    A=@n6.G
    B=MaterialFunctionCall(UVs=@n9, CenterPosition=@n10, Radius=@n11) {materialFunction=&RadialGradientExponential}.RadialGradientExponential
   )
   B=DynamicParameter() {parameterIndex=1, defaultValue=<rgba 0 1 1 25>, paramNames [Erode,SpeedX,SpeedY,HDRIntensity]}.Erode
  ) {constB=0.02}
  DepthFade(
   Opacity=Multiply(A=@n12, B=ParticleColor().A)
   FadeDistance=DynamicParameter() {parameterIndex=2, defaultValue=<rgba 0 1 1 25>, paramNames [DepthFade,---,---,---]}.DepthFade
  )
  n13 = Saturate(None=ParticleColor().RGB)
  n14 = DynamicParameter() {parameterIndex=1, defaultValue=<rgba 0 1 1 25>, paramNames [Erode,SpeedX,SpeedY,HDRIntensity]}
  MaterialFunctionCall(Emissive=Multiply(
    A=MaterialFunctionCall(LDR Input=@n13, HDR Intensity Multiplier=@n14.HDRIntensity) {materialFunction=&DeriveHDRfromLDR}.RemappedRange
    B=@n12
   )) {materialFunction=&MF_FX_EmissiveAdjustments}
  SmoothStep(Value=@n12)
 ]
}
