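# Translucent unlit fire material for Niagara sprite and mesh particles, built around a packed gas-torch flipbook texture.
# A SubUV mip-derivative function samples it using dynamic-parameter grid size and frame, with UVs built from the U coordinate on both axes.
# A Packed Channel switch is wired to blue on every branch, so selection has no effect.
# That channel multiplies particle color and alpha.
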Material /Game/Effects/Niagara/MasterMaterials/Fire/M_Fire_Stateless_Niagara
graph {
 properties {
  bAutomaticallySetUsageInEditor=false
  bUsedWithNiagaraMeshParticles=true
  bUsedWithNiagaraSprites=true
  blendMode=BLEND_Translucent
  shadingModel=MSM_Unlit
  usedShadingModels=Unlit
 }
 nodes [
  n1 = ParticleColor()
  n2 = DynamicParameter() {paramNames [Param1, Packed Channel, SubUV_Images, SubUV_Frame]}
  n3 = TextureObjectParameter() {group=Flame Channels, parameterName=Texture, sortPriority=1, texture=/Game/Effects/Fort_Effects/Textures/Fire/2DGasTorch_Flipbk_packed}
  n4 = MaterialFunctionCall(Float2=TextureCoordinate()) {materialFunction=&BreakOutFloat2Components}
  n5 = MaterialFunctionCall(Float3=MaterialFunctionCall!(
    Texture=@n3
    UVs=MaterialFunctionCall(X=@n4.R, Y=@n4.R) {materialFunction=&MakeFloat2}.Result
    SubImages=AppendVector(A=@n2.SubUV_Images, B=@n2.SubUV_Images)
    Frame=@n2.SubUV_Frame
   ) {materialFunction=/Engine/Functions/Engine_MaterialFunctions03/Texturing/SubUV_Function_MipDerivative}.RGB) {materialFunction=&BreakOutFloat3Components}
  n6 = If(A=@n2.Packed Channel, "A > B"=@n5.B, "A == B"=@n5.B, "A < B"=@n5.B) {constB=1, desc=pick flipbook channels, equalsThreshold=0.1}
  Multiply(A=@n1.RGB, B=@n6)
  Multiply(A=@n1.A, B=@n6)
 ]
}
