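# Unlit, translucent, two-sided master material for Niagara sprite and mesh particles, defaulting to a fire flipbook.
# Emissive uses particle color, optionally tinted by a desaturatable ColorTexture, through an emissive-adjustment function.
# Opacity comes from a noise-modulated flipbook, animated by scalar or dynamic parameter, with Use_ChangeTexture blending toward ColorTexture alpha.
# Particle alpha, depth fade and view-angle fade then attenuate the opacity.
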Material /Game/Effects/Niagara/MasterMaterials/Fire/M_Simple_FlipBookSubTex
graph {
 properties {
  bUsedWithNiagaraMeshParticles=true
  bUsedWithNiagaraSprites=true
  blendMode=BLEND_Translucent
  pixelDepthOffsetMode=PDOM_Legacy
  shadingModel=MSM_Unlit
  twoSided=true
  usedShadingModels=Unlit
 }
 nodes [
  n1 = ParticleColor()
  n2 = TextureSampleParameter2D() {parameterName=ColorTexture, texture=/Game/Effects/Fort_Effects/Textures/Smoke/T_SmokePuffy_SubUV_2X2}
  n3 = Desaturation(None=@n2.RGB, Fraction=ScalarParameter() {parameterName=Desaturation})
  n4 = ScalarParameter() {parameterName=Use_ChangeTexture}
  MaterialFunctionCall(Emissive=LinearInterpolate(A=@n1.RGB, B=Multiply(A=@n3, B=@n1.RGB), Alpha=@n4)) {materialFunction=&MF_FX_EmissiveAdjustments}
  n5 = LinearInterpolate(
   A=ScalarParameter() {parameterName="Animation  Phase (0-1) (S)"}
   B=DynamicParameter() {defaultValue=<rgba 0 1 1 1>, paramNames [Animation,Param2,Param3,Param4]}.Animation
   Alpha=ScalarParameter() {parameterName=UseDyAnim}
  )
  n6 = ScalarParameter() {defaultValue=2, parameterName="Number of Rows (S)"}
  n7 = ScalarParameter() {defaultValue=2, parameterName="Number of Columns (S)"}
  n8 = TextureObjectParameter() {parameterName=FlipBookTex, texture=/Game/Effects/Fort_Effects/Textures/T_MadameMoth_fire_sprites}
  n9 = ScalarParameter() {parameterName=LerpMin}
  n10 = ScalarParameter() {defaultValue=1, parameterName=LerpMax}
  n11 = VectorParameter() {parameterName=NoiseUVScale, defaultValue=<rgba 1 1 0 0>}
  n12 = TextureSampleParameter2D(UVs=Multiply(A=TextureCoordinate(), B=ComponentMask(None=@n11.RGB) {g=true, r=true})) {parameterName=Noise, samplerType=SAMPLERTYPE_Grayscale, texture=/Game/Weapons/FORT_Melee/Pickaxe_Cube_Ninja_Male/FX/T_CellNoise}
  n13 = LinearInterpolate(A=@n12.R, Alpha=ScalarParameter() {defaultValue=0.1, parameterName=Noise Lerp}) {constAlpha=0.1}
  n14 = LinearInterpolate(
   A=Multiply(
    A=MaterialFunctionCall("Animation  Phase (0-1)"=@n5, Number of Rows=@n6, Number of Columns=@n7, Texture=@n8) {materialFunction=&FlipBook}.Result
    B=LinearInterpolate(A=@n9, B=@n10, Alpha=@n13)
   )
   B=@n2.A
   Alpha=@n4
  )
  Multiply(
   A=DepthFade(Opacity=Multiply(A=@n1.A, B=@n14), FadeDistance=ScalarParameter() {defaultValue=20, parameterName=FadeDistance}) {fadeDistanceDefault=5}
   B=MaterialFunctionCall("FadeSensitivity (optional)"=ScalarParameter() {defaultValue=2, parameterName="FadeSensitivity (optional) (S)"}) {materialFunction=&FadeBasedOnViewAngle}.Result
  )
 ]
}
